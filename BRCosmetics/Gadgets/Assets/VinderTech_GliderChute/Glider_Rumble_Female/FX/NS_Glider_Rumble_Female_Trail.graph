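# Niagara trail effect for the Rumble Female glider: three smoke ribbons, rising embers, a flickering glow sprite with a point light, and speed lines.
# Embers, glow, light and ribbons sample the User skeletal mesh for position.
# ForwardDot, RightDot, a dive attribute, integrators and timelines blend ribbon velocities, flicker and ember spawning.
# Speed-line opacity follows ForwardDot in gameplay, or deployment state in front-end preview, which also lowers their density.
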NiagaraSystem /BRCosmetics/Gadgets/Assets/VinderTech_GliderChute/Glider_Rumble_Female/FX/NS_Glider_Rumble_Female_Trail
dependencies {
 usedDataInterfaces [&NiagaraDataInterfaceCurve, &NiagaraDataInterfaceColorCurve, &NiagaraDataInterfaceVector2DCurve]
 usedDynamicInputs [&SetBoolByFloatComparison]
 usedModules [
  &self:SystemSpawnScript.SystemScriptSource.SystemScriptGraph.NiagaraNodeAssignment_0.SetVariables_#1
  &SystemState
  &FN_Integrator
  &self:SystemSpawnScript.SystemScriptSource.SystemScriptGraph.NiagaraNodeAssignment_1.SetVariables_#2
  &self:SystemSpawnScript.SystemScriptSource.SystemScriptGraph.NiagaraNodeAssignment_2.SetVariables_#3
  &Timeline
 ]
}
emitters {
 Embers {
  emitterUpdateScript [
   &EmitterState(
    Enable Distance Culling=true
    Enable Visibility Culling=true
    MaxDistance=3000
    Module.bUseMaxDistance=true
    Module.bUseMinDistance=false
    Scalability Mode=::NewEnumerator1
   )
   &SpawnRate(
    Spawn Probability=&Lerp_Float(A=&UniformRangedFloat(Minimum=0, Maximum=1), B=1, Alpha=$System.FN_Integrator001.Value)
    SpawnRate=&UniformRangedFloat(Minimum=0, Maximum=5)
    Use Spawn Probability=true
   )
  ]
  particleSpawnScript [
   /Niagara/Modules/Spawn/Initialization/InitializeParticle(
    Lifetime=&UniformRangedFloat(Minimum=0.3, Maximum=2.5)
    Module.Write Color=true
    Module.Write Lifetime=true
    Module.Write Mass=true
    Module.Write Material Random=true
    Module.Write Position=true
    Module.Write Scale=false
    Module.Write SpriteRotation=false
    Module.Write SpriteSize=true
    Position=&SimulationPosition()
    Sprite Size=&Vector2DFromFloat(Value=&FloatFromCurve(
      FloatCurve=DataInterface() {propertyValues {Curve {keys [<at 0 1.2>,<at 0.1257 0.9847>,<at 0.2996 0.5>,<at 1 0.3>]}}}
      CurveIndex=&UniformRangedFloat(Minimum=0, Maximum=1)
      Scale Curve=2
     ))
   )
   &SkeletalMeshLocation(MaskSpawnGroup=false, Mesh Sampling Type=::NewEnumerator5, OverrideRandomSeed=false, Skeletal Mesh=$User.SkeletalMesh)
   &CylinderLocation(
    Cylinder Height=5
    Cylinder Origin:NiagaraPosition=&ConvertVectorToPosition(
     Input Position=Linked() {linkedVariable {name=Particles.SkeletalMeshLocation.SampledPosition, type=NiagaraPosition}}
     Vector Is In Absolute Large World Space:NiagaraBool=false
    )
    Cylinder Radius=15
    MaskSpawnGroup=false
    Override Seed=false
    UseOwnerScale=true
   )
   &AddVelocityFromPoint(Module.Use Falloff Distance=false, Origin Offset=<-150 -500 -150>, Velocity Strength=350)
  ]
  particleUpdateScript [
   &ParticleState()
   &AccelerationForce(Acceleration=&UniformRangedVector(Minimum=<35 0 150>, Maximum=<200 1 250>))
   &Drag(
    Drag=&FloatFromCurve(FloatCurve=DataInterface() {propertyValues {Curve {keys [<at 0 0.5>,<at 0.5 0.5>,<at 1 1>]}}}, Scale Curve=9)
    Use Linear Drag=true
    Use Rotational Drag=false
   )
   /Niagara/Modules/Update/Forces/CurlNoiseForce(
    Noise Frequency=0.04
    Noise Strength=&FloatFromCurve(FloatCurve=DataInterface() {propertyValues {Curve {keys [<at 0 1>,<at 1 0>]}}}, Scale Curve=1500)
    Pan Noise=true
    Pan Noise Field=<0 0 1>
   )
   &SolveForcesAndVelocity(Clamp Velocity=false, Limit Acceleration=false)
   &ScaleSpriteSizeBySpeed(
    Max Scale Factor=<2 5>
    Sample Scale Factor By Curve=true
    Scale Factor Curve=DataInterface() {propertyValues {Curve {keys [
        <at -0.0077 0.0765 RCIM_Cubic>
        {arriveTangent=1.2463, interpMode=RCIM_Cubic, leaveTangent=1.2463, tangentMode=RCTM_User, time=0.9985, value=0.9549}
       ]}}}
    Velocity Threshold=500
   )
   &ScaleColor(
    Scale Alpha=&FloatFromCurve(FloatCurve=DataInterface() {propertyValues {Curve {keys [<at 0 1>,<at 1 0>]}}})
    Scale RGB=&MakeVectorFromLinearColorRGB(Color=&ColorFromCurve(ColorCurve=DataInterface() {propertyValues {
        AlphaCurve {keys [<at 0 1 RCIM_Cubic>,<at 0.7107 1>]}
        BlueCurve {keys [<at 0 14.6701 RCIM_Cubic>,<at 0.2499 1.5104>,<at 0.7114 0.1753 RCIM_Cubic>]}
        GreenCurve {keys [<at 0 36.7481 RCIM_Cubic>,<at 0.2499 5.293>,<at 0.7114 0.3243 RCIM_Cubic>]}
        RedCurve {keys [<at 0 50 RCIM_Cubic>,<at 0.2499 10>,<at 0.7114 1 RCIM_Cubic>]}
       }}))
    ScaleA=true
    ScaleRGB=true
   )
  ]
  rendererClasses [&NiagaraSpriteRendererProperties]
  renderers [{rendererClass=&NiagaraSpriteRendererProperties}]
 }
 FN_Light {
  emitterUpdateScript [
   &EmitterState(
    Delay First Loop Only=true
    Inactive Response=::NewEnumerator1
    Life Cycle Mode=::NewEnumerator1
    Loop Delay=0.5
    Loop Duration=3.4
    UseLoopDelay=true
   )
   &SpawnBurst_Instantaneous(Spawn Count=1, Spawn Group=1, Spawn Probability=HlslExpression() {hlslExpression="Emitter.LoopCount == 0"}, Use Spawn Probability=true)
  ]
  particleSpawnScript [
   &InitializeParticle(
    Lifetime=3
    Module.Write Color=true
    Module.Write Lifetime=true
    Module.Write Mass=false
    Module.Write Material Random=true
    Module.Write Position=false
    Module.Write Scale=false
    Module.Write SpriteRotation=true
    Module.Write SpriteSize=true
    Position=&SimulationPosition()
    Sprite Rotation=&UniformRangedFloat(Minimum=0, Maximum=360)
    Sprite Size=<35 35>
   )
   &SkeletalMeshLocation(
    MaskSpawnGroup=false
    Mesh Sampling Type=::NewEnumerator5
    Sampled Position Offset Coordinate Space:ENiagaraSkelMeshTransforms=::NewEnumerator0
    Skeletal Mesh=$User.SkeletalMesh
    Socket Sampling Mode=::NewEnumerator5
   ) {module=SkeletalMeshLocation001}
  ]
  particleUpdateScript [
   &ParticleState(Kill Particles When Lifetime Has Elapsed=false, Let Infinitely Lived Particles Die When Emitter Deactivates=true)
   &SkeletalMeshLocation(MaskSpawnGroup=false, Mesh Sampling Type=::NewEnumerator5, Skeletal Mesh=$User.SkeletalMesh, Socket Sampling Mode=::NewEnumerator5)
   &ScaleColor(
    Scale Alpha=&Lerp_Float(
     A=&RandomRangeFloat(
      Minimum=&FloatFromCurve(
       FloatCurve=DataInterface() {propertyValues {Curve {keys [
           {interpMode=RCIM_Cubic, tangentMode=RCTM_User, time=0, value=0.4}
           {arriveTangent=-0.5149, interpMode=RCIM_Cubic, leaveTangent=-0.5149, time=0.1058, value=1.0057}
           {arriveTangent=-1.0417, interpMode=RCIM_Cubic, leaveTangent=-1.0417, time=0.2049, value=0.3945}
           {arriveTangent=6.4111, interpMode=RCIM_Cubic, leaveTangent=6.4111, time=0.3066, value=0.7965}
           {arriveTangent=3.0359, interpMode=RCIM_Cubic, leaveTangent=3.0359, time=0.4015, value=1.655}
           {interpMode=RCIM_Cubic, tangentMode=RCTM_User, time=0.5032, value=1.3934}
           {arriveTangent=0.4011, interpMode=RCIM_Cubic, leaveTangent=0.4011, time=0.6032, value=1.1665}
           {arriveTangent=-2.5522, interpMode=RCIM_Cubic, leaveTangent=-2.5522, time=0.7007, value=1.4726}
           {arriveTangent=-2.5332, interpMode=RCIM_Cubic, leaveTangent=-2.5332, time=0.8032, value=0.6561}
           {arriveTangent=-0.7932, interpMode=RCIM_Cubic, leaveTangent=-0.7932, time=0.9058, value=0.9531}
           <at 1 0.5 RCIM_Cubic>
          ]}}}
       CurveIndex=$Emitter.NormalizedLoopAge
       Scale Curve=0.75
      )
      Maximum=&FloatFromCurve(
       FloatCurve=DataInterface() {propertyValues {Curve {keys [
           {interpMode=RCIM_Cubic, tangentMode=RCTM_User, time=0, value=0.8}
           <at 0.1 0.3 RCIM_Cubic>
           {arriveTangent=2.403, interpMode=RCIM_Cubic, leaveTangent=2.403, time=0.2, value=0.8}
           {arriveTangent=0.75, interpMode=RCIM_Cubic, leaveTangent=0.75, time=0.3066, value=0.7965}
           {arriveTangent=-2.8258, interpMode=RCIM_Cubic, leaveTangent=-2.8258, time=0.4, value=0.95}
           {interpMode=RCIM_Cubic, tangentMode=RCTM_User, time=0.5, value=0.25}
           {arriveTangent=2, interpMode=RCIM_Cubic, leaveTangent=2, time=0.6, value=0.5}
           {arriveTangent=-1, interpMode=RCIM_Cubic, leaveTangent=-1, time=0.7, value=0.65}
           {arriveTangent=-0.75, interpMode=RCIM_Cubic, leaveTangent=-0.75, time=0.8, value=0.3}
           {arriveTangent=1, interpMode=RCIM_Cubic, leaveTangent=1, time=0.9, value=0.5}
           <at 1 0.8 RCIM_Cubic>
          ]}}}
       CurveIndex=$Emitter.NormalizedLoopAge
       Scale Curve=0.75
      )
     )
     B=1
     Alpha=$System.FN_Integrator001.Value
    )
    Scale RGB=&MakeVectorFromLinearColorRGB(Color=&ColorFromCurve(ColorCurve=DataInterface() {propertyValues {
        AlphaCurve {keys [<at 0 1 RCIM_Cubic>]}
        BlueCurve {keys [<at 0.0029 0.0879 RCIM_Cubic>,<at 0.4489 0.095>]}
        GreenCurve {keys [<at 0.0029 0.4606 RCIM_Cubic>,<at 0.4489 0.2181>]}
        RedCurve {keys [<at 0.0029 1 RCIM_Cubic>,<at 0.4489 1>]}
       }}))
    ScaleA=true
    ScaleRGB=true
   )
   &SolveForcesAndVelocity(Clamp Velocity=false, Limit Acceleration=false)
   &CameraOffset(Camera Offset Amount=10)
   &ScaleSpriteSize(Uniform Curve Sprite Scale=DataInterface() {propertyValues {Curve {keys [
        {arriveTangent=4.1785, interpMode=RCIM_Cubic, leaveTangent=4.1785, tangentMode=RCTM_User, time=0, value=0}
        {arriveTangent=0.9995, interpMode=RCIM_Cubic, leaveTangent=0.9995, tangentMode=RCTM_User, time=0.2223, value=0.6825}
        {arriveTangent=-0.8776, interpMode=RCIM_Cubic, leaveTangent=-0.8776, time=0.9, value=0.875}
        {arriveTangent=-10.3541, interpMode=RCIM_Cubic, leaveTangent=-10.3541, tangentMode=RCTM_User, time=1, value=1}
       ]}}})
  ]
  rendererClasses [&NiagaraSpriteRendererProperties]
  renderers [{rendererClass=&NiagaraSpriteRendererProperties}]
 }
 Light {
  emitterUpdateScript [
   &EmitterState()
   &SpawnBurst_Instantaneous(Spawn Count=1, Spawn Group=1, Spawn Probability=HlslExpression() {hlslExpression="Emitter.LoopCount == 0"}, Use Spawn Probability=true)
  ]
  particleSpawnScript [&InitializeParticle(
    Color=<rgba 1 0.321 0.0381 1>
    Lifetime=&UniformRangedFloat(Minimum=0.5, Maximum=1)
    Module.Write Color=true
    Module.Write Lifetime=false
    Module.Write Mass=true
    Module.Write Material Random=true
    Module.Write Position=false
    Module.Write Scale=false
    Module.Write SpriteRotation=false
    Module.Write SpriteSize=false
    Sprite Size=<8 5>
   )]
  particleUpdateScript [
   &FN_Light_v3(Create Exponent Attribute=true, Create Radius Attribute=true, Enable Particle Light Enable=true, Exponent=5, Light Enabled=true, Radius=100)
   &ParticleState(Kill Particles When Lifetime Has Elapsed=false, Let Infinitely Lived Particles Die When Emitter Deactivates=true)
   &ScaleColor(
    Scale Alpha=&Lerp_Float(
     A=&RandomRangeFloat(
      Minimum=&FloatFromCurve(
       FloatCurve=DataInterface() {propertyValues {Curve {keys [
           {interpMode=RCIM_Cubic, tangentMode=RCTM_User, time=0, value=0.5}
           {arriveTangent=-0.5149, interpMode=RCIM_Cubic, leaveTangent=-0.5149, time=0.1058, value=1.0057}
           {arriveTangent=-1.0417, interpMode=RCIM_Cubic, leaveTangent=-1.0417, time=0.2049, value=0.3945}
           {arriveTangent=6.4111, interpMode=RCIM_Cubic, leaveTangent=6.4111, time=0.3066, value=0.7965}
           {arriveTangent=3.0359, interpMode=RCIM_Cubic, leaveTangent=3.0359, time=0.4015, value=1.655}
           {interpMode=RCIM_Cubic, tangentMode=RCTM_User, time=0.5032, value=1.3934}
           {arriveTangent=0.4011, interpMode=RCIM_Cubic, leaveTangent=0.4011, time=0.6032, value=1.1665}
           {arriveTangent=-2.5522, interpMode=RCIM_Cubic, leaveTangent=-2.5522, time=0.7007, value=1.4726}
           {arriveTangent=-2.5332, interpMode=RCIM_Cubic, leaveTangent=-2.5332, time=0.8032, value=0.6561}
           {arriveTangent=-0.7932, interpMode=RCIM_Cubic, leaveTangent=-0.7932, time=0.9058, value=0.9531}
           <at 1 0.5 RCIM_Cubic>
          ]}}}
       CurveIndex=$Emitter.NormalizedLoopAge
       Scale Curve=0.75
      )
      Maximum=&FloatFromCurve(
       FloatCurve=DataInterface() {propertyValues {Curve {keys [
           {interpMode=RCIM_Cubic, tangentMode=RCTM_User, time=0, value=0.8}
           <at 0.1 0.3 RCIM_Cubic>
           {arriveTangent=2.403, interpMode=RCIM_Cubic, leaveTangent=2.403, time=0.2, value=0.8}
           {arriveTangent=0.75, interpMode=RCIM_Cubic, leaveTangent=0.75, time=0.3066, value=0.7965}
           {arriveTangent=-2.8258, interpMode=RCIM_Cubic, leaveTangent=-2.8258, time=0.4, value=0.95}
           {interpMode=RCIM_Cubic, tangentMode=RCTM_User, time=0.5, value=0.25}
           {arriveTangent=2, interpMode=RCIM_Cubic, leaveTangent=2, time=0.6, value=0.5}
           {arriveTangent=-1, interpMode=RCIM_Cubic, leaveTangent=-1, time=0.7, value=0.65}
           {arriveTangent=-0.75, interpMode=RCIM_Cubic, leaveTangent=-0.75, time=0.8, value=0.3}
           {arriveTangent=1, interpMode=RCIM_Cubic, leaveTangent=1, time=0.9, value=0.5}
           <at 1 0.8 RCIM_Cubic>
          ]}}}
       CurveIndex=$Emitter.NormalizedLoopAge
       Scale Curve=0.75
      )
     )
     B=2
     Alpha=$System.FN_Integrator001.Value
    )
    Scale RGB=&VectorFromFloat(Value=2)
    ScaleA=true
    ScaleRGB=true
   )
   &SkeletalMeshLocation(
    MaskSpawnGroup=false
    Mesh Sampling Type=::NewEnumerator5
    Sampled Position Offset Coordinate Space:ENiagaraSkelMeshTransforms=::NewEnumerator4
    Skeletal Mesh=$User.SkeletalMesh
    Socket Sampling Mode=::NewEnumerator5
   )
  ]
  rendererClasses [&NiagaraLightRendererProperties]
  renderers [{rendererClass=&NiagaraLightRendererProperties, rendererIndex=0}]
 }
 Smoke_Trail {
  emitterUpdateScript [
   &EmitterState()
   /Game/Effects/Niagara/Modules/Skydiving/FN_DivingAttributes/FN_EmitterDivingAttributes_v1(Bone Index=0, SkeletalMesh:NiagaraDataInterfaceSkeletalMesh=$User.SkeletalMesh)
   &FN_SpawnRibbon_v2(SpawnRate=16)
  ]
  particleSpawnScript [
   &FN_InitializeRibbon_v3(
    Color:LinearColor=/Niagara/DynamicInputs/TypeConversions/MakeLinearColorFromFloat(Value:NiagaraFloat=1)
    Enable Skeletal Position=true
    Lifetime=2.75
    Ribbon Width=10
    Sampled Position Coordinate Space:ENiagaraCoordinateSpace=::Local
    Sampled Rotation Coordinate Space:ENiagaraCoordinateSpace=::Local
    Sampled Velocity Coordinate Space:ENiagaraCoordinateSpace=::Local
   ) {module=FN_InitializeRibbon_v3001}
   &AddVelocity(
    Coordinate Space=::World
    Scale Added Velocity=&Lerp_Vector(A=<1 1 1>, B=<1 1 2>, Alpha=$Emitter.Dive)
    Velocity=&Lerp_Vector(A=<0 0 100>, B=<0 0 30>, Alpha=$User.ForwardDot)
   ) {module=AddVelocity003}
   &AddVelocity(
    Scale Added Velocity=&Lerp_Vector(A=<1 1 1>, B=<1 1 2>, Alpha=$Emitter.Dive)
    Velocity=&Lerp_Vector(
     A=&Lerp_Vector(A=&Lerp_Vector(A=<35 50 0>, B=<75 50 1>, Alpha=$User.ForwardDot), B=<-50 50 1>, Alpha=$System.Timeline004.Time)
     B=<50 50 0>
     Alpha=$System.Timeline005.Time
    )
   )
   &SampleSkeletalMeshSkeleton(Mesh:NiagaraDataInterfaceSkeletalMesh=$User.SkeletalMesh, Spawn Mode:ENiagaraSkeletalSpawning=::NewEnumerator1)
   /SetVariables(Particles.RibbonU0Override=&Multiply_Float(A=$Emitter.Age, B=2)) {isSetParameters=true}
   &AddVelocity(Velocity=&FN_FrontEndVector3Modifier(Gameplay=<0 0 0>, Front End=&RandomRangeVector(Minimum=<0 150 -90>, Maximum=<0 300 -100>))) {module=AddVelocity001}
  ]
  particleUpdateScript [
   /Niagara/Noise/Waveform(
    EnableExponent1=false
    Global Amplitude Scale=15
    GlobalClampMax=false
    GlobalClampMin=false
    UseInput1=false
    UsePhase1=false
    "Waveform [1]":ENiagara_Waveforms=::NewEnumerator12
    "[1] Amplitude Min/Max":Vector2f=<-0.1 0.1>
    "[1] Frequency"=10
    "[1] Phase Scale"=2
    "[1] Scale"=2
   )
   /SetVariables(Particles.Velocity=&Add_Vector(
     A=$Particles.Velocity
     B=&MakeVector(
      X=&Add_Float(
       A=$Particles.Waveform.Waveform
       B=&Lerp_Float(
        A=&Lerp_Float(
         A=&Lerp_Float(A=0, B=&RandomRangeFloat(Minimum=0, Maximum=2), Alpha=$User.ForwardDot)
         B=5
         Alpha=$System.FN_Integrator.Value
        )
        B=-5
        Alpha=$User.RightDot
       )
      )
      Y=&Lerp_Float(A=&RandomRangeFloat(Minimum=-0.1, Maximum=0.1), B=8, Alpha=$User.ForwardDot)
      Z=&Lerp_Float(
       A=0
       B=&FloatFromCurve(
        FloatCurve=DataInterface() {propertyValues {Curve {keys [
            {interpMode=RCIM_Cubic, tangentMode=RCTM_User, time=0, value=-2}
            {arriveTangent=6.1333, interpMode=RCIM_Cubic, leaveTangent=6.1333, tangentMode=RCTM_User, time=1, value=2}
           ]}}}
        Scale Curve=&RandomRangeFloat(Minimum=0.8, Maximum=1)
       )
       Alpha=$User.ForwardDot
      )
     )
    )) {isSetParameters=true}
   &SolveForcesAndVelocity(Clamp Velocity=false, Limit Acceleration=false)
   &FN_UpdateRibbon(
    Enable Skeletal Position=true
    Initial Color:LinearColor=$Particles.Initial.Color
    Initial RibbonWidth:NiagaraFloat=$Particles.Initial.RibbonWidth
   )
   &Color(
    Color=&ColorFromCurve(ColorCurve=DataInterface() {propertyValues {
       AlphaCurve {keys [<at 0 2>]}
       BlueCurve {keys [<at 0 0>,<at 0.0301 0.6667>,<at 0.4709 0.1525>]}
       GreenCurve {keys [<at 0 0>,<at 0.0301 0.5969>,<at 0.4709 0.1704>]}
       RedCurve {keys [<at 0 50>,<at 0.0301 0.6251>,<at 0.4709 0.2552>]}
      }})
    Scale Alpha=&FloatFromCurve(FloatCurve=DataInterface() {propertyValues {Curve {keys [
         {arriveTangent=0.7543, interpMode=RCIM_Cubic, leaveTangent=0.7543, tangentMode=RCTM_User, time=0, value=0}
         <at 0.1 0.05>
         {arriveTangent=-0.013, interpMode=RCIM_Cubic, leaveTangent=-0.013, tangentMode=RCTM_User, time=0.2992, value=0.0096}
         {interpMode=RCIM_Cubic, leaveTangent=-0.2611, time=1, value=0}
        ]}}})
   )
   &Material(
    Base Texture Time:NiagaraFloat=&Multiply_Float(A=$Emitter.Age, B=&Lerp_Float(A=-0.35, B=-0.1, Alpha=$User.ForwardDot))
    PostMult=0.6
    Power=0.5
   )
  ]
  rendererClasses [&NiagaraRibbonRendererProperties]
  renderers [{rendererClass=&NiagaraRibbonRendererProperties}]
 }
 Smoke_Trail001 {
  emitterUpdateScript [
   &EmitterState()
   &FN_EmitterDivingAttributes_v1(Bone Index=0, SkeletalMesh:NiagaraDataInterfaceSkeletalMesh=$User.SkeletalMesh)
   &FN_SpawnRibbon_v2(SpawnRate=12)
  ]
  particleSpawnScript [
   &FN_InitializeRibbon_v3(
    Color:LinearColor=&MakeLinearColorFromFloat(Value:NiagaraFloat=1)
    Enable Skeletal Position=true
    Lifetime=2.5
    Sampled Position Coordinate Space:ENiagaraCoordinateSpace=::Local
    Sampled Rotation Coordinate Space:ENiagaraCoordinateSpace=::Local
    Sampled Velocity Coordinate Space:ENiagaraCoordinateSpace=::Local
   ) {module=FN_InitializeRibbon_v3001}
   &AddVelocity(
    Coordinate Space=::World
    Scale Added Velocity=&Lerp_Vector(A=<1 1 1>, B=<1 1 2>, Alpha=$Emitter.Dive)
    Velocity=&Lerp_Vector(A=<0 0 100>, B=<0 0 30>, Alpha=$User.ForwardDot)
   ) {module=AddVelocity003}
   &AddVelocity(
    Scale Added Velocity=&Lerp_Vector(A=<1 1 1>, B=<1 1 2>, Alpha=$Emitter.Dive)
    Velocity=&Lerp_Vector(
     A=&Lerp_Vector(A=&Lerp_Vector(A=<35 50 0>, B=<75 50 1>, Alpha=$User.ForwardDot), B=<-50 50 1>, Alpha=$System.Timeline004.Time)
     B=<50 50 0>
     Alpha=$System.Timeline005.Time
    )
   )
   &SampleSkeletalMeshSkeleton(Mesh:NiagaraDataInterfaceSkeletalMesh=$User.SkeletalMesh, Spawn Mode:ENiagaraSkeletalSpawning=::NewEnumerator1)
   /SetVariables(Particles.RibbonU0Override=&Multiply_Float(A=$Emitter.Age, B=2)) {isSetParameters=true}
   &AddVelocity(Velocity=&FN_FrontEndVector3Modifier(Gameplay=<0 0 0>, Front End=&RandomRangeVector(Minimum=<0 150 -90>, Maximum=<0 300 -100>))) {module=AddVelocity001}
  ]
  particleUpdateScript [
   &Waveform(
    EnableExponent1=false
    Global Amplitude Scale=20
    GlobalClampMax=false
    GlobalClampMin=false
    UseInput1=false
    UsePhase1=false
    "Waveform [1]":ENiagara_Waveforms=::NewEnumerator12
    "[1] Amplitude Min/Max":Vector2f=<-0.2 0.2>
    "[1] Frequency"=5
    "[1] Phase Scale"=2
   )
   /SetVariables(Particles.Velocity=&Add_Vector(
     A=$Particles.Velocity
     B=&MakeVector(
      X=&Add_Float(
       A=$Particles.Waveform.Waveform
       B=&Lerp_Float(
        A=&Lerp_Float(
         A=&Lerp_Float(A=0, B=&RandomRangeFloat(Minimum=0, Maximum=1), Alpha=$User.ForwardDot)
         B=5
         Alpha=$System.FN_Integrator.Value
        )
        B=-5
        Alpha=$User.RightDot
       )
      )
      Y=&Lerp_Float(A=&RandomRangeFloat(Minimum=-0.1, Maximum=0.1), B=10, Alpha=$User.ForwardDot)
      Z=&Lerp_Float(
       A=0
       B=&FloatFromCurve(
        FloatCurve=DataInterface() {propertyValues {Curve {keys [
            {interpMode=RCIM_Cubic, tangentMode=RCTM_User, time=0, value=-2}
            {arriveTangent=6.1333, interpMode=RCIM_Cubic, leaveTangent=6.1333, tangentMode=RCTM_User, time=1, value=5}
           ]}}}
        Scale Curve=&RandomRangeFloat(Minimum=0.25, Maximum=0.35)
       )
       Alpha=$User.ForwardDot
      )
     )
    )) {isSetParameters=true, module=SetVariables_#4}
   &SolveForcesAndVelocity(Clamp Velocity=false, Limit Acceleration=false)
   &FN_UpdateRibbon(
    Enable Skeletal Position=true
    Initial Color:LinearColor=$Particles.Initial.Color
    Initial RibbonWidth:NiagaraFloat=$Particles.Initial.RibbonWidth
   )
   &Color(
    Color=&ColorFromCurve(ColorCurve=DataInterface() {propertyValues {
       AlphaCurve {keys [<at 0 1>]}
       BlueCurve {keys [<at 0 0>,<at 0.0301 0.6667>,<at 0.4709 0.1525>]}
       GreenCurve {keys [<at 0 0>,<at 0.0301 0.5969>,<at 0.4709 0.1704>]}
       RedCurve {keys [<at 0 50>,<at 0.0301 0.6251>,<at 0.4709 0.2552>]}
      }})
    Scale Alpha=&FloatFromCurve(FloatCurve=DataInterface() {propertyValues {Curve {keys [
         {arriveTangent=0.7543, interpMode=RCIM_Cubic, leaveTangent=0.7543, tangentMode=RCTM_User, time=0, value=0}
         <at 0.1 0.05>
         {arriveTangent=-0.013, interpMode=RCIM_Cubic, leaveTangent=-0.013, tangentMode=RCTM_User, time=0.2992, value=0.0096}
         {interpMode=RCIM_Cubic, leaveTangent=-0.2611, time=1, value=0}
        ]}}})
   )
   &Material(
    Base Texture Time:NiagaraFloat=&Multiply_Float(A=$Emitter.Age, B=&Lerp_Float(A=-0.25, B=-0.1, Alpha=$User.ForwardDot))
    PostMult=0.6
    Power=0.5
   )
  ]
  rendererClasses [&NiagaraRibbonRendererProperties]
  renderers [{rendererClass=&NiagaraRibbonRendererProperties}]
 }
 Smoke_Trail002 {
  emitterUpdateScript [
   &EmitterState()
   &FN_EmitterDivingAttributes_v1(Bone Index=0, SkeletalMesh:NiagaraDataInterfaceSkeletalMesh=$User.SkeletalMesh)
   &FN_SpawnRibbon_v2(SpawnRate=12)
  ]
  particleSpawnScript [
   &FN_InitializeRibbon_v3(
    Color:LinearColor=&MakeLinearColorFromFloat(Value:NiagaraFloat=1)
    Enable Skeletal Position=true
    Lifetime=3
    Ribbon Width=7.5
    Sampled Position Coordinate Space:ENiagaraCoordinateSpace=::Local
    Sampled Rotation Coordinate Space:ENiagaraCoordinateSpace=::Local
    Sampled Velocity Coordinate Space:ENiagaraCoordinateSpace=::Local
   ) {module=FN_InitializeRibbon_v3001}
   &AddVelocity(
    Coordinate Space=::World
    Scale Added Velocity=&Lerp_Vector(A=<1 1 1>, B=<1 1 2>, Alpha=$Emitter.Dive)
    Velocity=&Lerp_Vector(A=<0 0 100>, B=<0 0 30>, Alpha=$User.ForwardDot)
   ) {module=AddVelocity003}
   &AddVelocity(
    Scale Added Velocity=&Lerp_Vector(A=<1 1 1>, B=<1 1 2>, Alpha=$Emitter.Dive)
    Velocity=&Lerp_Vector(
     A=&Lerp_Vector(A=&Lerp_Vector(A=<35 50 0>, B=<75 50 1>, Alpha=$User.ForwardDot), B=<-50 50 1>, Alpha=$System.Timeline004.Time)
     B=<50 50 0>
     Alpha=$System.Timeline005.Time
    )
   )
   &SampleSkeletalMeshSkeleton(Mesh:NiagaraDataInterfaceSkeletalMesh=$User.SkeletalMesh, Spawn Mode:ENiagaraSkeletalSpawning=::NewEnumerator1)
   /SetVariables(Particles.RibbonU0Override=&Multiply_Float(A=$Emitter.Age, B=2)) {isSetParameters=true}
   &AddVelocity(Velocity=&FN_FrontEndVector3Modifier(Gameplay=<0 0 0>, Front End=&RandomRangeVector(Minimum=<0 200 -90>, Maximum=<0 400 -100>))) {module=AddVelocity001}
  ]
  particleUpdateScript [
   &Waveform(
    EnableExponent1=false
    Global Amplitude Scale=25
    GlobalClampMax=false
    GlobalClampMin=false
    UseInput1=false
    UsePhase1=false
    "Waveform [1]":ENiagara_Waveforms=::NewEnumerator12
    "[1] Amplitude Min/Max":Vector2f=<-0.1 0.1>
    "[1] Frequency"=3.5
    "[1] Phase Scale"=1.5
   )
   /SetVariables(Particles.Velocity=&Add_Vector(
     A=$Particles.Velocity
     B=&MakeVector(
      X=&Add_Float(
       A=$Particles.Waveform.Waveform
       B=&Lerp_Float(
        A=&Lerp_Float(
         A=&Lerp_Float(A=0, B=&RandomRangeFloat(Minimum=0, Maximum=1), Alpha=$User.ForwardDot)
         B=5
         Alpha=$System.FN_Integrator.Value
        )
        B=-5
        Alpha=$User.RightDot
       )
      )
      Y=&Lerp_Float(A=&RandomRangeFloat(Minimum=-0.1, Maximum=0.1), B=10, Alpha=$User.ForwardDot)
      Z=&Lerp_Float(
       A=0
       B=&FloatFromCurve(
        FloatCurve=DataInterface() {propertyValues {Curve {keys [
            {interpMode=RCIM_Cubic, tangentMode=RCTM_User, time=0, value=-2}
            {arriveTangent=6.1333, interpMode=RCIM_Cubic, leaveTangent=6.1333, tangentMode=RCTM_User, time=1, value=10}
           ]}}}
        Scale Curve=&RandomRangeFloat(Minimum=0.1, Maximum=-0.1)
       )
       Alpha=$User.ForwardDot
      )
     )
    )) {isSetParameters=true, module=SetVariables_#4}
   &SolveForcesAndVelocity(Clamp Velocity=false, Limit Acceleration=false)
   &FN_UpdateRibbon(
    Enable Skeletal Position=true
    Initial Color:LinearColor=$Particles.Initial.Color
    Initial RibbonWidth:NiagaraFloat=$Particles.Initial.RibbonWidth
   )
   &Color(
    Color=&ColorFromCurve(ColorCurve=DataInterface() {propertyValues {
       AlphaCurve {keys [<at 0 1>]}
       BlueCurve {keys [<at 0 0>,<at 0.0301 0.6667>,<at 0.4094 0.177>]}
       GreenCurve {keys [<at 0 0>,<at 0.0301 0.5969>,<at 0.4094 0.2002>]}
       RedCurve {keys [<at 0 50>,<at 0.0301 0.6251>,<at 0.4094 0.2552>]}
      }})
    Scale Alpha=&FloatFromCurve(FloatCurve=DataInterface() {propertyValues {Curve {keys [
         {arriveTangent=0.7543, interpMode=RCIM_Cubic, leaveTangent=0.7543, tangentMode=RCTM_User, time=0, value=0}
         <at 0.1 0.05>
         {arriveTangent=-0.013, interpMode=RCIM_Cubic, leaveTangent=-0.013, tangentMode=RCTM_User, time=0.2992, value=0.0096}
         {interpMode=RCIM_Cubic, leaveTangent=-0.2611, time=1, value=0}
        ]}}})
   )
   &Material(
    Base Texture Time:NiagaraFloat=&Multiply_Float(A=$Emitter.Age, B=&Lerp_Float(A=-0.5, B=-0.1, Alpha=$User.ForwardDot))
    PostMult=0.6
    Power=0.5
   )
  ]
  rendererClasses [&NiagaraRibbonRendererProperties]
  renderers [{rendererClass=&NiagaraRibbonRendererProperties}]
 }
 SpeedLine {
  emitterUpdateScript [
   &EmitterState()
   &SpawnRate(
    SpawnRate=&MakeCustomFloatFromBool(Bool=$User.bIsFrontEndPreview, True Float=50, False Float=90)
    Use Spawn Probability=true
   ) {module=SpawnRate001}
  ]
  particleSpawnScript [
   /Niagara/Modules/Spawn/Initialization/V2/InitializeParticle(
    Color=&MakeLinearColorFromVectorAndFloat(
     "Vector (RGB)"=<1 1 1>
     "Float (Alpha)"=&Multiply_Float(
      A=&UniformRangedFloat(Minimum=0.1, Maximum=0.18)
      B=&MakeCustomFloatFromBool(Bool=$User.bIsFrontEndPreview, True Float=0.5, False Float=1)
     )
    )
    Lifetime Max=0.16
    Lifetime Min=0.1
    Lifetime Mode=::NewEnumerator1
    Sprite Size Max=<3 10>
    Sprite Size Min=<2 2>
    Sprite Size Mode=::NewEnumerator2
    UsePositionOffset=false
    Write Material Random=true
   )
   &TorusLocation(
    Align Torus With Axis=true
    Handle Radius=0
    Large Radius:NiagaraFloat=&RandomRangeFloat(Minimum=80, Maximum=150)
    MaskSpawnGroup=false
    Module.Override Randomness=false
    Offset=<0 -180 0>
    Torus Axis=<0 1 0>
   )
   &ShapeLocation(
    Handle Radius=1
    Large Radius=&RandomRangeFloat(Minimum=80, Maximum=150)
    MaskSpawnGroup=false
    Non Uniform Scale=<1.5 1 0.1>
    Override Seed=false
    Shape Primitive=::NewEnumerator3
    UseOwnerScale=true
   )
   &AddVelocityFromPoint(Module.Use Falloff Distance=false, Origin Offset=<0 -500 0>, Velocity Strength=600)
   &AddVelocity(Velocity=&RandomRangeVector(Minimum=<0 300 0>, Maximum=<1 700 1>))
  ]
  particleUpdateScript [
   &SolveForcesAndVelocity(Clamp Velocity=false, Limit Acceleration=false)
   &ParticleState()
   &ScaleSpriteSize(Scale Factor=&Vector2DFromCurve(
     Vector2Curve=DataInterface() {propertyValues {XCurve {keys [<at 0 0.5 RCIM_Cubic>,<at 0.5 1 RCIM_Cubic>]}, YCurve {keys [<at 0 0 RCIM_Cubic>,<at 0.5 1 RCIM_Cubic>]}}}
     ScaleCurve=&Vector2DFromFloat(Value=1)
    ))
   &ScaleSpriteSizeBySpeed(Max Scale Factor=<1 50>, Min Scale Factor=<0.01 0.05>, Velocity Threshold=1)
   &ScaleColor(
    Scale Alpha=&FloatFromCurve(
     FloatCurve=DataInterface() {propertyValues {Curve {keys [
         <at 0 0 RCIM_Cubic>
         <at 0.3759 1.0207>
         {arriveTangent=-2.221, interpMode=RCIM_Cubic, leaveTangent=-2.221, tangentMode=RCTM_User, time=1, value=0}
        ]}}}
     Scale Curve=&MakeCustomFloatFromBool(
      Bool=$User.bIsFrontEndPreview
      True Float=&MakeCustomFloatFromBool(Bool=$User.bIsFullyDeployed, True Float=1, False Float=0)
      False Float=&Lerp_Float(A=0, B=1, Alpha=$User.ForwardDot)
     )
    )
    Scale RGB=&VectorFromFloat(Value=1)
    ScaleA=true
    ScaleRGB=true
   )
  ]
  rendererClasses [&NiagaraSpriteRendererProperties]
  renderers [{rendererClass=&NiagaraSpriteRendererProperties}]
 }
}
summary {
 systemName=NS_Glider_Rumble_Female_Trail
 emitters [
  {emitterName=Smoke_Trail, rendererClasses [&NiagaraRibbonRendererProperties]}
  {emitterName=Light, rendererClasses [&NiagaraLightRendererProperties]}
  {emitterName=Embers, rendererClasses [&NiagaraSpriteRendererProperties]}
  {emitterName=FN_Light, rendererClasses [&NiagaraSpriteRendererProperties]}
  {emitterName=Smoke_Trail001, rendererClasses [&NiagaraRibbonRendererProperties]}
  {emitterName=Smoke_Trail002, rendererClasses [&NiagaraRibbonRendererProperties]}
  {emitterName=SpeedLine, rendererClasses [&NiagaraSpriteRendererProperties]}
 ]
 userVariables [
  User.bIsFrontEnd {type=NiagaraBool}
  User.bIsFrontEndPreview {type=NiagaraBool}
  User.bIsFullyDeployed {type=NiagaraBool}
  User.ForwardDot {type=NiagaraFloat, defaultValue=0, description=""}
  User.LightIntensityScale {type=NiagaraFloat, defaultValue=0, description=""}
  User.LightScale {type=NiagaraFloat, defaultValue=0, description=""}
  User.RibbonIDOffset {type=NiagaraInt32}
  User.RightDot {type=NiagaraFloat, defaultValue=0, description=""}
  User.SkeletalMesh {
   type=NiagaraDataInterfaceSkeletalMesh
   defaultValue {
    struct=NiagaraExt_VariableValue_DataInterface
    value {dataInterface=&self:NiagaraDataInterfaceSkeletalMesh_0}
   }
  }
 ]
}
